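# Conversation database for the HybridBoot NPC, who greets players as the maknae of HUNTR/X.
# The entry point leads to one speech node whose contextual lines react to a first-ever conversation, match conversation count or specific player outfits.
# The match-count checks reference the Specialist.Medic.EvokeFind spawner tag.
# The speech then links by tag to shared NPC quests, basic services and HybridBoot's own dialogue entry.
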ConversationDatabase /NPCLibrary/NPCs/HybridBoot/NPCLibrary_Conversation_HybridBoot
entry_points [{tag=Conversation.Entry.HybridBoot, destinations [#1]}]
nodes [
 ConversationEntryPointNode_7 {
  id=#1
  class=ConversationEntryPointNode
  outputs [#2]
  properties {entryTag {tagName=Conversation.Entry.HybridBoot}}
 }
 Speech_C_4 {
  id=#2
  class=Speech_C
  outputs [#3, #4, #5]
  properties {
   nodeName=""
   generalConfig {
    defaultMessage="If you're looking for the maknae of HUNTR/X, you got her!"
    contextualMessages [
     {
      message="If you're looking for the maknae of HUNTR/X, you got her!"
      contextRequirements [{requirement {class=FortControllerRequirement_IsFirstEverConversationWithNPC}}]
      priority {curve {rowName=Tandem.Speech.Priority.FirstEverInteraction}}
     }
     {
      message="Happy fans, happy Honmoon."
      contextRequirements [{requirement {
         class=FortControllerRequirement_CheckNPCMatchConversationCount
         properties {nPCUniqueIdTag {tagName=AISpawnerData.Type.Specialist.Medic.EvokeFind}}
        }}]
      priority {curve {rowName=None}}
     }
     {
      message="HUNTR/X forever!"
      contextRequirements [{requirement {
         class=FortControllerRequirement_CheckNPCMatchConversationCount
         properties {nPCUniqueIdTag {tagName=AISpawnerData.Type.Specialist.Medic.EvokeFind}}
        }}]
      priority {curve {rowName=None}}
     }
     {
      message="Couch! Couch! Couch! Couch!"
      contextRequirements [{requirement {
         class=FortControllerRequirement_HasCID
         properties {allowedCIds [/BRCosmetics/Athena/Items/Cosmetics/Characters/Character_QuailWink, /BRCosmetics/Athena/Items/Cosmetics/Characters/Character_LowerDrop]}
        }}]
      priority {curve {rowName=Tandem.Speech.Priority.OutfitReaction}}
     }
     {
      message="So, what's the plan?"
      contextRequirements [{requirement {
         class=FortControllerRequirement_HasCID
         properties {allowedCIds [/BRCosmetics/Athena/Items/Cosmetics/Characters/Character_LoanFloat, /BRCosmetics/Athena/Items/Cosmetics/Characters/Character_TurnVortex]}
        }}]
      priority {curve {rowName=Tandem.Speech.Priority.OutfitReaction}}
     }
     {
      message=Just because Rumi trusts you don't mean I do.
      contextRequirements [{requirement {class=FortControllerRequirement_HasCID, properties {allowedCIds [/BRCosmetics/Athena/Items/Cosmetics/Characters/Character_BountyPond]}}}]
      priority {curve {rowName=Tandem.Speech.Priority.OutfitReaction}}
     }
    ]
   }
  }
 }
 ConversationLinkNode_2 {
  id=#5
  class=ConversationLinkNode
  properties {remoteEntryTag {tagName=Conversation.Entry.HybridBoot.Dialogue}}
 }
 ConversationLinkNode_0 {id=#4, class=ConversationLinkNode, properties {remoteEntryTag {tagName=Conversation.Entry.BasicServices}}}
 ConversationLinkNode_1 {id=#3, class=ConversationLinkNode, properties {remoteEntryTag {tagName=Conversation.Entry.NPCLibrary.Quests}}}
]
